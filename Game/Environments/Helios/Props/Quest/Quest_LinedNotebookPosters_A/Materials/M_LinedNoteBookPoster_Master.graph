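# Master material for a quest poster prop, splitting poster and frame textures by UV1, with an animated LED-screen emissive on the poster.
# Adaptive cell UVs drive an LCD mask, multiplied by blended icon patterns and two-color emissive with hash flicker and scrolling scanlines.
# A noise SphereMask transition follows timestamp parameters, or loops on time through a test switch.
# One packed green channel feeds Metallic, Specular and Roughness.
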Material /Game/Environments/Helios/Props/Quest/Quest_LinedNotebookPosters_A/Materials/M_LinedNoteBookPoster_Master
graph {
 properties {bUseMaterialAttributes=true, bUsedWithNanite=true, pixelDepthOffsetMode=PDOM_Legacy, usedShadingModels=Default Lit}
 nodes [
  Constant() {r=-100}
  n1 = ScalarParameter() {parameterName=EdgeOffset}
  n2 = ScalarParameter() {defaultValue=1.2, parameterName=uvscale}
  n3 = ScalarParameter() {defaultValue=0.008, parameterName=CellSize}
  n4 = TextureSample(UVs=NamedRerouteDeclaration(None=MaterialFunctionCall!(bump, offsetToEdge=@n1, uvscale=@n2, PixelAspect, cellSize=@n3) {materialFunction=/Game/Environments/Asteria/Sets/NeonCity/Techart/Screen/MaterialFunctions/MF_AdaptiveCellsUVs}.tilingUVs) {name=ScaledUVs, nodeColor=<rgba 1 0 0.0938 1>}) {samplerType=SAMPLERTYPE_Masks, texture=/Game/Environments/Asteria/Sets/NeonCity/Techart/Textures/T_LCD_Panels_LCDSoft}
  n5 = LinearInterpolate(
   A=TextureSampleParameter2D() {parameterName=Pattern1, samplerType=SAMPLERTYPE_Masks, texture=/Game/Environments/FrontEnd/FX/Stars/Textures/T_StarIcon_01}.RGB
   B=LinearInterpolate(
    A=TextureSampleParameter2D() {parameterName=Pattern2, texture=/Game/UI/Foundation/Textures/Banner/BattleRoyale/Season_3/T-Banners-Icons-ShootingStar-L}.R
    B=TextureSampleParameter2D() {parameterName=Patter2_b, texture=/Game/UI/Foundation/Textures/Banner/BattleRoyale/Season_3/T-Banners-Icons-Sun}.R
    Alpha=ScalarParameter() {parameterName=random seed}
   )
   Alpha=NamedRerouteUsage()
  )
  n6 = VectorParameter() {parameterName=Emissive, defaultValue=<rgba 0.3 0.01 1.2 1>}
  n7 = VectorParameter() {parameterName=Emissive2, defaultValue=<rgba 0 0.2188 0.0268 1>}
  n8 = SmoothStep(Value=NamedRerouteUsage()) {constMin=0.5}
  n9 = Multiply(A=Time(), B=ScalarParameter() {defaultValue=0.5, parameterName=Emissive_FlickerSpeed})
  n10 = Multiply(
   A=Saturate(None=MaterialFunctionCall(Float=@n9) {materialFunction=&MF_Hash}.Float)
   B=ScalarParameter() {defaultValue=0.4, parameterName=Emissive_FlickerAmount}
  ) {constB=0.3}
  n11 = Multiply(A=Panner(Coordinate=TextureCoordinate()) {speedY=0.02}) {constB=40}
  n12 = SmoothStep(Value=ComponentMask(None=Frac(None=@n11)) {g=true}) {constMax=0.9, constMin=0.5}
  n13 = Multiply(A=Multiply(A=LinearInterpolate(A=@n6.RGB, B=@n7.RGB, Alpha=@n8)) {constB=8}, B=Multiply(A=Add(A=@n10), B=Add(A=@n12)))
  n14 = If(
   A=Constant() {r=1}
   B=ComponentMask(None=TextureCoordinate() {coordinateIndex=1}) {r=true}
   "A > B"=Multiply(A=@n4.RGB, B=Multiply(A=@n5, B=@n13))
   "A < B"=Constant()
  ) {desc=0}
  LinearInterpolate(A=@n14, B=Constant() {r=10}, Alpha=Sign(None=ScalarParameter() {parameterName=TimeStampA}))
  n15 = TextureCoordinate() {coordinateIndex=1}
  n16 = TextureSample() {
   samplerType=SAMPLERTYPE_Normal
   texture=/Game/Environments/Asteria/Props/Commerce/Commerce_SidewalkSigns_A/Textures/T_Commerce_SidewalkSigns_Frame_A_N
  }
  NamedRerouteDeclaration(None=If(A=Constant() {r=1}, B=ComponentMask(None=@n15) {r=true}, "A > B"=Constant3Vector() {constant=<rgba 0 0 1 0>}, "A < B"=@n16.RGB) {desc=0}) {name=Normal_OUT, nodeColor=<rgba 0.7578 1 0 1>}
  n17 = TextureCoordinate() {coordinateIndex=1}
  NamedRerouteDeclaration(None=If(
    A=Constant() {r=1}
    B=ComponentMask(None=@n17) {r=true}
    "A > B"=TextureSample() {texture=/Game/Environments/Asteria/Props/Commerce/Commerce_SidewalkSigns_A/Textures/T_Commerce_SideWalkSigns_Poster_A_D}.RGB
    "A < B"=TextureSample() {texture=/Game/Environments/Asteria/Props/Commerce/Commerce_SidewalkSigns_A/Textures/T_Commerce_SidewalkSigns_Frame_A_D}.RGB
   ) {desc=0}) {name=Diffused_OUT, nodeColor=<rgba 0 1 0.25 1>}
  n18 = TextureCoordinate() {coordinateIndex=1}
  n19 = TextureSample() {
   samplerType=SAMPLERTYPE_Masks
   texture=/Game/Environments/Asteria/Props/Commerce/Commerce_SidewalkSigns_A/Textures/T_Commerce_SideWalkSigns_Poster_A_S
  }
  n20 = TextureSample() {
   samplerType=SAMPLERTYPE_Masks
   texture=/Game/Environments/Asteria/Props/Commerce/Commerce_SidewalkSigns_A/Textures/T_Commerce_SidewalkSigns_Frame_A_S
  }
  NamedRerouteDeclaration(None=If(A=Constant() {r=1}, B=ComponentMask(None=@n18) {r=true}, "A > B"=@n19.RGB, "A < B"=@n20.RGB) {desc=0}) {name=Spec_OUT, nodeColor=<rgba 1 0 0.3984 1>}
  NamedRerouteDeclaration(None=@n14) {name=Emissive_OUT, nodeColor=<rgba 0.8516 1 0 1>}
  n21 = Multiply(A=Multiply(A=TextureCoordinate(), B=Constant2Vector() {g=1.5, r=1}), B=ScalarParameter() {defaultValue=1, parameterName=NoiseTiling})
  n22 = Time()
  n23 = Frac(None=Multiply(A=Add(A=@n22) {constB=0.4}, B=ScalarParameter() {defaultValue=1, parameterName=TransitionSpeed}) {constB=0.6})
  n24 = ScalarParameter() {parameterName=TimeStampA}
  n25 = Abs(None=Sign(None=@n24))
  n26 = MaterialFunctionCall(
   TimestampA=@n24
   TransitionLength_In=Constant() {r=1}
   TimestampB=ScalarParameter() {parameterName=TimeStampB}
   TransitionLength_Out=Constant() {r=1}
  ) {materialFunction=&MF_Timestamp_InOut_01}
  n27 = StaticSwitchParameter(True=Saturate(None=Multiply(A=@n23) {constB=1.4}), False=Power(Base=Multiply(A=@n25, B=@n26.Linear)) {constExponent=1.3}) {parameterName=TestTransition}
  NamedRerouteDeclaration(None=SphereMask(
    A=TextureSampleParameter2D(UVs=@n21) {parameterName=Noise, samplerType=SAMPLERTYPE_LinearColor, texture=/Landmass/TextureCreation/Textures/TNoise1024}.R
    B=Constant() {r=-0.01}
    Radius=Multiply(A=@n27) {constB=1.05}
   ) {attenuationRadius=0.1}) {name=Transition, nodeColor=<rgba 0.2891 0 1 1>}
  NamedRerouteDeclaration(None=@n4.R) {name=LED, nodeColor=<rgba 0.1484 0 1 1>}
  n28 = MaterialFunctionCall(Float3=NamedRerouteUsage()) {materialFunction=&BreakOutFloat3Components}
  SetMaterialAttributes!(
   MaterialAttributes
   Base Color=NamedRerouteUsage()
   Metallic=@n28.G
   Specular=@n28.G
   Roughness=@n28.G
   Normal=NamedRerouteUsage()
   Emissive Color=NamedRerouteUsage()
  ) {attributeSetTypes [
    #1
    #2
    #3
    #4
    #5
    #6
   ]}
  Subtract(A=@n27) {constB=10}
 ]
}
